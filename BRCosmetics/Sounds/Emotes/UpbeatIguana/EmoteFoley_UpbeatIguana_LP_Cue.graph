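# Looping foley Sound Cue for the UpbeatIguana emote, using emote attenuation, concurrency and play-when-silent virtualization.
# A one-shot Start wave plays immediately, and the Main loop enters after a fixed delay timed to follow it.
# In parallel, High, Low and Wind loops sit behind a short delay and volume envelope, each with continuous pitch and volume modulation driven by the EmoteMoveSpeed parameter.
# The cue also has a send to the music audio-analysis submix.
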SoundCue /BRCosmetics/Sounds/Emotes/UpbeatIguana/EmoteFoley_UpbeatIguana_LP_Cue
root=Mixer_2
nodes [
 {id=Delay_0, children [WavePlayer_4], properties {delayMax=1.755, delayMin=1.755}}
 {id=Delay_1, children [Mixer_0], properties {delayMax=0.5, delayMin=0.5}}
 {
  id=Enveloper_1
  children [Delay_1]
  properties {pitchCurve {editorCurveData {keys [{time=0}]}}, volumeCurve {editorCurveData {keys [<at 0 0>,<at 0.5 1>]}}}
 }
 {id=Mixer_0, children [ModulatorContinuous_0, ModulatorContinuous_1, null, ModulatorContinuous_3], properties {inputVolume [1,1,1,1]}}
 {id=Mixer_1, children [WavePlayer_3,Delay_0], properties {inputVolume [1,1]}}
 {id=Mixer_2, children [Mixer_1,Enveloper_1], properties {inputVolume [1,1]}}
 {
  id=ModulatorContinuous_0
  children [WavePlayer_2]
  properties {pitchModulationParams {minOutput=0.9, parameterName=EmoteMoveSpeed}, volumeModulationParams {minOutput=0.2, parameterName=EmoteMoveSpeed}}
 }
 {
  id=ModulatorContinuous_1
  children [WavePlayer_1]
  properties {
   pitchModulationParams {maxOutput=1.1, minOutput=1, parameterName=EmoteMoveSpeed}
   volumeModulationParams {minOutput=0.4, parameterName=EmoteMoveSpeed}
  }
 }
 {
  id=ModulatorContinuous_3
  children [WavePlayer_0]
  properties {pitchModulationParams {minOutput=1, parameterName=EmoteMoveSpeed}, volumeModulationParams {maxOutput=0.8, parameterName=EmoteMoveSpeed}}
 }
 {id=WavePlayer_0, properties {bLooping=true, soundWaveAssetPtr=&/IP_Emote_UpbeatIguana_LP_Wind}}
 {id=WavePlayer_1, properties {bLooping=true, soundWaveAssetPtr=&/IP_Emote_UpbeatIguana_LP_Low}}
 {id=WavePlayer_2, properties {bLooping=true, soundWaveAssetPtr=&/IP_Emote_UpbeatIguana_LP_High}}
 {id=WavePlayer_3, properties {soundWaveAssetPtr=&/IP_Emote_UpbeatIguana_Start}}
 {id=WavePlayer_4, properties {bLooping=true, soundWaveAssetPtr=&/IP_Emote_UpbeatIguana_LP_Main}}
]
properties {
 attenuationSettings=&Fort_Emotes_Athena_Att
 audioPropertiesSheet=/BRCosmetics/Sounds/Settings/PropertySheets/Emotes/APS_Cosmetic_Emote_Foley
 maxDistance=1600
 soundClassObject=/Game/Sounds/FortSoundClassesAndModes/Fort_Foley_Emote
 virtualizationMode=PlayWhenSilent
 volumeMultiplier=1
 concurrencyOverrides {maxCount=3}
 concurrencySet [/Game/Sounds/Concurrency_Rules/Athena_EmoteFoley_Conc]
 soundSubmixSends [{minSendLevel=0, soundSubmix=&Fortnite_AudioAnalysis_Music}]
}
